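; Player clambering component that handles the presentation and gameplay side of climbing ledges on a Fortnite pawn.
; Targeting events drive a highlight decal, in-range tag and input prompt, while gameplay tags gate targeting and activation.
; Synched action start picks a montage by height and plays it through a root motion source, with camera shake and gameplay cue.
; The server applies a gameplay effect and cancels or blocks abilities, undoing effect and block when ledge movement stops.
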(event EventBeginPlay
  ;[0 1 0.04714] Store the player pawn
  (bind _asfort_player_pawn (Utilities|Casting|CastToFortPlayerPawn (Components|GetOwner)))
  (bind _output_get (Variables|Default|SetPlayerPawn _asfort_player_pawn))
  (CallFunction|RegisterEvents _output_get))

(event EventEndPlay
  (params (EndPlayReason :type EEndPlayReason%20Enum :default "Destroyed"))
  (bind _playerpawn (Variables|Default|GetPlayerPawn)
    (:then
      (Ability|GameplayEffect|RemoveLooseGameplayTags _playerpawn (Variables|Default|GetT_TargetWithinRange)))
    (:else)))

(event Custom|CustomEvent_0
  (bind _asfort_pawn (Utilities|Casting|CastToFortPawn (Components|GetOwner)))
  (if (Utilities|Enum|Equal(Enum) (Networking|GetLocalRole _asfort_pawn) "ROLE_AutonomousProxy")
    (bind _asfort_player_controller (Utilities|Casting|CastToFortPlayerController (Pawn|GetController _asfort_pawn)))
    (Input|SetRotatePawntoCamera _asfort_player_controller (not (Game|Player|IsInputKeyDown _asfort_player_controller "U")))
    (Variables|/Script/Engine.Pawn|SetbUseControllerRotationYaw (not (Game|Player|IsInputKeyDown _asfort_player_controller "U")) _asfort_pawn)))

(event EventHandleTargetingDataValid
  (params (TargetingData :type Clambering%20Targeting%20Data%20Structure%20%28by%20ref%29))
  (CallFunction|ShouldTutorialbeEnabled)
  (if (CallFunction|ShouldShowClamberIndicator)
    (ClamberLocationIndicator|UpdateClamberLocationHighlight TargetingData)))

(event EventHandleTargetingDataInvalid
  (ClamberLocationIndicator|DestroyClamberLocationHighlight))

(event EventHandleClamberingTargetInActivationRange
  (CallFunction|SetUISceneVisible true)
  ;[0.08581 0 1] Apply/remove tag to player to show they're within range.
  (bind _playerpawn (Variables|Default|GetPlayerPawn)
    (:then
      (Ability|GameplayEffect|AddLooseGameplayTags _playerpawn (Variables|Default|GetT_TargetWithinRange)))
    (:else)))

(event EventHandleClamberingTargetOutOfActivationRange
  (CallFunction|SetUISceneVisible false)
  ;[0.08581 0 1] Apply/remove tag to player to show they're within range.
  (bind _playerpawn (Variables|Default|GetPlayerPawn)
    (:then
      (Ability|GameplayEffect|RemoveLooseGameplayTags _playerpawn (Variables|Default|GetT_TargetWithinRange)))
    (:else)))

(event EventHandleClamberingStateChanged
  (params (OldClamberingState :type EClamberingState%20Enum :default "Inactive") (NewClamberingState :type EClamberingState%20Enum :default "Inactive"))
  (bind _asfort_pawn (Utilities|Casting|CastToFortPawn (Components|GetOwner)))
  (CallFunction|HandleClamberStateStopped _asfort_pawn OldClamberingState)
  (CallFunction|HandleClamberStateStarted _asfort_pawn NewClamberingState))

(event Custom|HandleClamberStateStarted
  (params (Pawn :type Fort%20Pawn%20Object%20Reference) (ClamberingState :type EClamberingState%20Enum :default "Inactive"))
  (switch Utilities|FlowControl|Switch|SwitchonEClamberingState ClamberingState
    (:Inactive
      ;[1 1 1] Handle Clambering Reset
      (Weapon|SetHolsterWeapon Pawn 0 false false "Clamber"))
    (:Ledge_Failed
      ;[1 1 1] Handle Starting Clambering on a Ledge - FAILED
      (Animation|PlayAnimMontage Pawn (Variables|Default|GetClamberFailAnim) 1.0 "None"))))

(event EventHandleSynchedActionStarted
  (params (SynchedActionInfo :type Synched%20Action%20Info%20Structure%20%28by%20ref%29))
  (bind (_type _bvalid _bcanstandonledge _biswindow _sourceactor _sourcelocation _sourceaim _walllocation _wallnormal _targetlocation) (Utilities|Struct|BreakClamberingTargetingData (Variables|ClamberingComponent|GetLockedTargetingData)))
  (bind _asfort_pawn (Utilities|Casting|CastToFortPawn (Components|GetOwner)))
  (if (not (GameplayTags|HasMatchingGameplayTag _asfort_pawn (Variables|Default|GetTag_PreventWeaponHolster)))
    (Weapon|SetHolsterWeapon _asfort_pawn 0 true false "Clamber")
    (bind _clamber_anim (CallFunction|GetClamberAnim)))
  (Utilities|IsValid _clamber_anim
    (:Is%20Valid
      (RootMotionSources|PlayMontagewithRootMotionSource _asfort_pawn _clamber_anim (CallFunction|GetClamberAnimPlayRate) true)
      (Game|Feedback|ClientStartCameraShake :self (Fortnite|GetFortPlayerControllerfromActor _asfort_pawn) :Shake (Variables|Default|GetCameraShake) :Scale 0.5 :PlaySpace "CameraLocal")
      (GameplayCue|ExecuteGameplayCueOnActor(Burst) _asfort_pawn (Variables|Default|GetClamberGameplayCue) (Ability|GameplayCue|MakeGameplayCueParameters :NormalizedMagnitude 0.0 :RawMagnitude 0.0 :Location _targetlocation :Normal "0,0,0" :GameplayEffectLevel 1 :AbilityLevel 1 :bReplicateLocationWhenUsingMinimalRepProxy false))
      (CallFunction|SetUISceneVisible false)
      (Networking|SwitchHasAuthority
        (:Authority
          (bind _asfort_ability_system_component (Utilities|Casting|CastToFortAbilitySystemComponent (Ability|GetAbilitySystemComponent _asfort_pawn)))
          (bind _returnvalue (GameplayEffects|FortApplyGameplayEffectToSelf :self _asfort_ability_system_component :GameplayEffectClass (Variables|Default|GetGE_Clambering) :Level 0.0))
          (Variables|Default|SetActiveGEHandle _returnvalue)
          (Tags|CancelAbilitiesWithTags (Ability|GetFortAbilitySystemComponent _asfort_pawn) (GameplayTags|MakeGameplayTagContainerfromArray (Variables|Default|GetCancelAbilitiesWithTags)))
          (Tags|BlockAbilitiesWithTags (Ability|GetFortAbilitySystemComponent _asfort_pawn) (GameplayTags|MakeGameplayTagContainerfromArray (Variables|Default|GetBlockAbilitiesWithTags))))
        (:Remote
          (Utilities|FlowControl|Delay 0.25)
          (Actor|DestroyActor (Variables|Default|GetClamberLocationHighlight)))))))

(event Custom|HandleClamberStateStopped
  (params (Pawn :type Fort%20Pawn%20Object%20Reference) (ClamberingState :type EClamberingState%20Enum :default "Inactive"))
  (switch Utilities|FlowControl|Switch|SwitchonEClamberingState ClamberingState
    (:Ledge_Moving
      ;[1 1 1] Handle Stopped Clambering on a Ledge
      (Animation|StopAnimMontage Pawn (Variables|Default|GetClamber_250up))
      (Networking|SwitchHasAuthority
        (:Authority
          (bind _asfort_ability_system_component (Utilities|Casting|CastToFortAbilitySystemComponent (Ability|GetAbilitySystemComponent Pawn)))
          (GameplayEffects|RemoveActiveGameplayEffect _asfort_ability_system_component (Variables|Default|GetActiveGEHandle) -1)
          (Variables|Default|SetActiveGEHandle)
          (Tags|UnBlockAbilitiesWithTags (Ability|GetFortAbilitySystemComponent Pawn) (GameplayTags|MakeGameplayTagContainerfromArray (Variables|Default|GetBlockAbilitiesWithTags)) false))))))

(event EventOnMMEStarted
  (CallFunction|SetUISceneVisible false))

(event Custom|OnUsingMovementToJump
  (params (bUsingMovementToJump :type Boolean))
  (Variables|Default|SetbUsingMovementToJump bUsingMovementToJump))

(fn CanStartClambering
  (bind _asfort_pawn (Utilities|Casting|CastToFortPawn (Components|GetOwner)))
  (if (GameplayTags|HasMatchingGameplayTag _asfort_pawn (Variables|Default|GetT_ClamberBlocked))
    (return false)
    (else
      (if (GameplayTags|HasMatchingGameplayTag _asfort_pawn (Variables|Default|GetT_JetpackEquipped))
        (return (or (GameplayTags|HasMatchingGameplayTag _asfort_pawn "(TagName=\"MovementMode.Falling\")") (GameplayTags|HasMatchingGameplayTag _asfort_pawn "(TagName=\"MovementMode.Parachuting\")")))
        (else
          (return true))))))

(fn CanStartTargeting
  (bind _asfort_player_pawn (Utilities|Casting|CastToFortPlayerPawn (Components|GetOwner)))
  (if (not (or (GameplayTags|HasMatchingGameplayTag _asfort_player_pawn "(TagName=\"Gameplay.Action.Player.DanceStunned\")") (GameplayTags|HasMatchingGameplayTag _asfort_player_pawn (Variables|Default|GetT_ClamberBlocked))))
    (Utilities|IsValid (Class|FortPlayerPawn|GetDBNOHoistee _asfort_player_pawn)
      (:Is%20Valid
        (if (GameplayTags|HasMatchingGameplayTag _asfort_player_pawn (Variables|Default|GetTag_AllowClamberWhileDBNOCarrying))
          (bind _returnvalue (Fortnite|HeldObjects|GetHeldObjectComponentAttachedtoPlayer _asfort_player_pawn))
          (Utilities|IsValid _returnvalue
            (:Is%20Valid
              (if (not (GameplayTags|HasMatchingGameplayTag _asfort_player_pawn (Variables|Default|GetTag_AllowClamberWhileHoldingObject)))
                (return false)))
            (:Is%20Not%20Valid
              _))
          (return true)
          (else
            _)))
      (:Is%20Not%20Valid
        _))))

(fn TutorialModeDisabled)

(fn TutorialModeEnabled)

(fn BuildDataRegistryResolverScope
  (return 0 0 false))

(fn CanShowUIPrompt
  (return (and (not (Networking|IsDedicatedServer)) (CallFunction|ShouldShowUIPrompt))))

(fn CreateClamberLocationHighlight
  (params (TargetingData :type Clambering%20Targeting%20Data%20Structure%20%28by%20ref%29))
  (bind (_transform _rotation) (ClamberLocationIndicator|GetClamberLocationHighlightTransform TargetingData))
  (bind _returnvalue (Components|GetOwner))
  (Utilities|IsValid (Variables|Default|GetClamberLocationHighlight)
    (:Is%20Not%20Valid
      (bind _aspawn (Utilities|Casting|CastToPawn _returnvalue))
      (bind _returnvalue_2 (Game|SpawnActorfromClass "/ClamberingContent/Gameplay/Assets/B_Decal_ClamberHighlight.B_Decal_ClamberHighlight_C" _transform "AlwaysSpawn" "MultiplyWithRoot" _returnvalue _aspawn))
      (Variables|Default|SetClamberLocationHighlight _returnvalue_2)
      (Transformation|SetWorldRotation (Variables|/Script/Engine.Actor|GetRootComponent _returnvalue_2) _rotation false false))))

(fn DestroyClamberLocationHighlight
  (bind _clamberlocationhighlight (Variables|Default|GetClamberLocationHighlight))
  (Utilities|IsValid _clamberlocationhighlight
    (:Is%20Valid
      (Actor|DestroyActor _clamberlocationhighlight)
      (Variables|Default|SetClamberLocationHighlight 0))))

(fn GetClamberAnim
  (bind _asfort_pawn (Utilities|Casting|CastToFortPawn (Components|GetOwner)))
  (bind _asfort_movement_comp_character (Utilities|Casting|CastToFortMovementComp_Character (Pawn|GetMovementComponent _asfort_pawn)))
  (bind (_synchedactioninfo_warpsyncpoint1name _synchedactioninfo_warpsyncpointlandingorientationname _synchedactioninfo_distancetotarget _synchedactioninfo_distancetotargetx _synchedactioninfo_distancetotargety _synchedactioninfo_distancetotargetz) (Utilities|Struct|BreakSynchedActionInfo (Variables|/Script/FortniteGame.FortMovementComp_Character|GetSynchedActionInfo_WarpSyncPoint1Name _asfort_movement_comp_character)))
  (if (>= _synchedactioninfo_distancetotargetz 1e+02)
    (return (Variables|Default|GetClamber_250up))
    (elif (>= _synchedactioninfo_distancetotargetz 2e+01)
      (return (Variables|Default|GetClamber_150up))
      (else
        (return (Variables|Default|GetClamber_60up))))))

(fn GetClamberAnimPlayRate
  (bind _clamber_250up (Variables|Default|GetClamber_250up))
  (Utilities|IsValid _clamber_250up
    (:Is%20Valid
      (return (Math|Float|SafeDivide (Animation|GetPlayLength _clamber_250up) (+ (ScalableFloat|GetValueatLevel (Utilities|Struct|BreakClamberingMovementConfigLedge (Variables|LedgeClamberingConfig|Movement|GetMoveConfig_Ledge))) 0.1))))
    (:Is%20Not%20Valid
      (return 1.0))))

(fn GetClamberLocationHighlightTransform
  (params (TargetingData :type Clambering%20Targeting%20Data%20Structure%20%28by%20ref%29))
  (bind (_type _bvalid _bcanstandonledge _biswindow _sourceactor _sourcelocation _sourceaim _walllocation _wallnormal _targetlocation _targetnormal) (Utilities|Struct|BreakClamberingTargetingData TargetingData))
  (return (Math|Transform|MakeTransform _targetlocation "0, 0, 0" "0.100000,0.100000,0.100000") (Math|Rotator|MakeRotfromX _targetnormal)))

(fn RegisterEvents
  (params (FortPlayerPawn :type Fort%20Player%20Pawn%20Object%20Reference))
  (bind _asfort_player_controller_gameplay (Utilities|Casting|CastToFortPlayerControllerGameplay (Pawn|GetController FortPlayerPawn)))
  (Utilities|IsValid _asfort_player_controller_gameplay
    (:Is%20Valid
      (EventDispatchers|BindEventtoOnUsingMovementtoJump _asfort_player_controller_gameplay (EventDispatchers|CreateEvent :!event OnUsingMovementToJump)))))

(fn SetUISceneVisible
  (params (bVisible :type Boolean))
  (bind _buisceneadded (Variables|Default|GetbUISceneAdded))
  (if (CallFunction|CanShowUIPrompt)
    (if bVisible
      (if (not _buisceneadded)
        (Variables|Default|SetbUISceneAdded true)
        (bind _asfort_pawn (Utilities|Casting|CastToFortPawn (Components|GetOwner)))
        (bind _asathena_player_controller (Utilities|Casting|CastToAthena_PlayerController (Pawn|GetController _asfort_pawn)))
        (Class|FortPlayerController|NotifyAvailableActionTarget _asathena_player_controller (select (Input|IsUsingGamepad _asathena_player_controller) (select (or (Variables|Default|GetbUsingMovementToJump) (CallFunction|IsAutoClamberingEnabled)) "MoveForward_Gamepad" "GamepadJump") (select (or (Variables|Default|GetbUsingMovementToJump) (CallFunction|IsAutoClamberingEnabled)) "MoveForward" "Jump")) (Variables|Default|GetPromptText)))
      (elif _buisceneadded
        (Variables|Default|SetbUISceneAdded false)
        (bind _asfort_pawn_2 (Utilities|Casting|CastToFortPawn (Components|GetOwner)))
        (bind _asathena_player_controller_2 (Utilities|Casting|CastToAthena_PlayerController (Pawn|GetController _asfort_pawn_2)))
        (Class|FortPlayerController|NotifyUnavailableActionTarget _asathena_player_controller_2)))))

(fn ShouldTutorialBeEnabled
  (bind _asfort_pawn (Utilities|Casting|CastToFortPawn (Components|GetOwner)))
  (if (Utilities|Enum|Equal(Enum) (Networking|GetLocalRole _asfort_pawn) "ROLE_AutonomousProxy")
    (if (GameplayTags|HasMatchingGameplayTag _asfort_pawn "(TagName=\"Granted.Athena.Tutorial.Clamber\")")
      (CallFunction|SetTutorialModeEnabled true)
      (else
        (CallFunction|SetTutorialModeEnabled false)))))

(fn UpdateClamberLocationHighlight
  (params (TargetingData :type Clambering%20Targeting%20Data%20Structure%20%28by%20ref%29))
  (bind (_transform _rotation) (ClamberLocationIndicator|GetClamberLocationHighlightTransform TargetingData))
  (bind _clamberlocationhighlight (Variables|Default|GetClamberLocationHighlight))
  (Utilities|IsValid (Variables|Default|GetClamberLocationHighlight)
    (:Is%20Valid
      _)
    (:Is%20Not%20Valid
      (ClamberLocationIndicator|CreateClamberLocationHighlight TargetingData)))
  (Transformation|SetActorTransform _clamberlocationhighlight _transform false false)
  (Transformation|SetWorldRotation (Variables|/Script/Engine.Actor|GetRootComponent _clamberlocationhighlight) _rotation false false))
